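# Animation Blueprint for the M_MED_Summer_Body_01 player body mesh, adding secondary motion to its shark-tooth necklace and pack.
# It copies the attached parent mesh's pose, then chains Anim Dynamics through the tooth, three necklace bones and pack, limited to LOD threshold 1.
# Spine-driven spherical limits constrain the tooth and first necklace bone.
# A neck Pose Driver, alternate necklace nodes and the stored MeshToCopy variable do not feed the output.
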AnimBlueprint /Game/Characters/Player/Male/Medium/Bodies/M_MED_SummerHeroes_01/Meshes/M_MED_Summer_Body_01_AnimBp
target_skeleton=&/M_MED_Summer_Body_01_Skeleton
graphs [
 AnimGraph {nodes [
   n1 = Anim Dynamics - Bone: dyn_necklace_02(Alpha=0.3, ExternalForce="0,0,0") {
    angularBiasOverride=1
    angularDampingOverride=0.98
    bGravityOverrideInSimSpace=true
    bUseSphericalLimits=true
    lODThreshold=1
    numSolverIterationsPostUpdate=4
    numSolverIterationsPreUpdate=16
    alphaBoolBlend {blendOption=Linear}
    boundBone {boneName=dyn_necklace_02}
    physicsBodyDefinitions [{
      sphereCollisionRadius=0
      boundBone {boneName=dyn_necklace_02}
      boxExtents=<5 5 10>
      constraintSetup {angularTargetAxis=AxisZ, twistAxis=AxisZ, angularLimitsMax=<80 45 0>, angularLimitsMin=<0 -45 0>, angularTarget=<0 0 0>}
      localJointOffset=<0 1 -1>
     }]
    sphericalLimits [{limitRadius=13.5, drivingBone {boneName=spine_04}, sphereLocalOffset=<0 2.25 0>}]
   }
   Anim Dynamics - Bone: dyn_necklace_03(ComponentPose=@n1, Alpha=0.3, ExternalForce="0,0,0") {
    angularBiasOverride=1
    angularDampingOverride=0.98
    bGravityOverrideInSimSpace=true
    bUseSphericalLimits=true
    lODThreshold=1
    numSolverIterationsPostUpdate=4
    numSolverIterationsPreUpdate=16
    alphaBoolBlend {blendOption=Linear}
    boundBone {boneName=dyn_necklace_03}
    physicsBodyDefinitions [{
      sphereCollisionRadius=0
      boundBone {boneName=dyn_necklace_03}
      boxExtents=<5 5 10>
      constraintSetup {angularTargetAxis=AxisZ, twistAxis=AxisZ, angularLimitsMax=<80 45 0>, angularLimitsMin=<0 -45 0>, angularTarget=<0 0 0>}
      localJointOffset=<0 1 -1>
     }]
    sphericalLimits [{limitRadius=13.5, drivingBone {boneName=spine_03}, sphereLocalOffset=<2 3 0>}]
   }
   Anim Dynamics - Bone: dyn_necklace_01(Alpha=1, ExternalForce="0,0,0") {
    angularBiasOverride=1
    angularDampingOverride=0.98
    bGravityOverrideInSimSpace=true
    bUseSphericalLimits=true
    lODThreshold=1
    numSolverIterationsPostUpdate=4
    numSolverIterationsPreUpdate=16
    alphaBoolBlend {blendOption=Linear}
    boundBone {boneName=dyn_necklace_01}
    physicsBodyDefinitions [{
      sphereCollisionRadius=0
      boundBone {boneName=dyn_necklace_01}
      boxExtents=<20 3 20>
      constraintSetup {
       angularTargetAxis=AxisZ
       twistAxis=AxisZ
       angularLimitsMax=<80 45 0>
       angularLimitsMin=<0 -45 0>
       angularTarget=<0 0 0>
       linearAxesMax=<0 1.5 0>
      }
      localJointOffset=<0 0 -1.5>
     }]
    sphericalLimits [{limitRadius=13.5, drivingBone {boneName=spine_04}, sphereLocalOffset=<1 0 0>}]
   }
   Pose Driver - Source: neck_01!(SourcePose:Pose Link Structure) -> (Pose:Pose Link Structure) {
    is=title AnimGraphNode_PoseDriver
    poseAsset=&/M_MED_Summer_Necklace_PoseAsset_01
    onlyDriveBones [{boneName=dyn_sharkTooth_01}]
    poseTargets [
     {
      bApplyCustomCurve=false
      bIsHidden=false
      distanceMethod=DefaultMethod
      drivenName=Pose_0
      functionType=DefaultFunction
      targetScale=1
      boneTransforms [{targetRotation=<rot 0 23.1062 0>, targetTranslation=<11.3138 0.1529 0>}]
      customCurve {defaultValue=3.4028234663852886e+38, postInfinityExtrap=RCCE_Constant, preInfinityExtrap=RCCE_Constant, keys []}
      targetRotation=<rot 0 0 0>
     }
     {
      bApplyCustomCurve=false
      bIsHidden=false
      distanceMethod=DefaultMethod
      drivenName=Pose_1
      functionType=DefaultFunction
      targetScale=1
      boneTransforms [{targetRotation=<rot 0 -66.8937 0>, targetTranslation=<11.3138 0.1529 0>}]
      customCurve {defaultValue=3.4028234663852886e+38, postInfinityExtrap=RCCE_Constant, preInfinityExtrap=RCCE_Constant, keys []}
      targetRotation=<rot 0 0 0>
     }
     {
      bApplyCustomCurve=false
      bIsHidden=false
      distanceMethod=DefaultMethod
      drivenName=Pose_2
      functionType=DefaultFunction
      targetScale=1
      boneTransforms [{targetRotation=<rot -90 0 23.1063>, targetTranslation=<11.3138 0.1529 0>}]
      customCurve {defaultValue=3.4028234663852886e+38, postInfinityExtrap=RCCE_Constant, preInfinityExtrap=RCCE_Constant, keys []}
      targetRotation=<rot 0 0 0>
     }
     {
      bApplyCustomCurve=false
      bIsHidden=false
      distanceMethod=DefaultMethod
      drivenName=Pose_3
      functionType=DefaultFunction
      targetScale=1
      boneTransforms [{targetRotation=<rot 0 113.1062 0>, targetTranslation=<11.3138 0.1529 0>}]
      customCurve {defaultValue=3.4028234663852886e+38, postInfinityExtrap=RCCE_Constant, preInfinityExtrap=RCCE_Constant, keys []}
      targetRotation=<rot 0 0 0>
     }
     {
      bApplyCustomCurve=false
      bIsHidden=false
      distanceMethod=DefaultMethod
      drivenName=Pose_4
      functionType=DefaultFunction
      targetScale=1
      boneTransforms [{targetRotation=<rot 90 0 -23.1063>, targetTranslation=<11.3138 0.1529 0>}]
      customCurve {defaultValue=3.4028234663852886e+38, postInfinityExtrap=RCCE_Constant, preInfinityExtrap=RCCE_Constant, keys []}
      targetRotation=<rot 0 0 0>
     }
    ]
    rBFParams {distanceMethod=Quaternion, function=Cubic, radius=90}
    sourceBones [{boneName=neck_01}]
   }
   n2 = Anim Dynamics - Bone: dyn_sharkTooth_01(ComponentPose=Local To Component(LocalPose=Copy Pose From Mesh() {bUseAttachedParent=true}), Alpha=1, ExternalForce="0,0,0") {
    angularBiasOverride=1
    angularDampingOverride=0.98
    bGravityOverrideInSimSpace=true
    bOverrideAngularBias=true
    bOverrideAngularDamping=true
    bUsePlanarLimit=false
    bUseSphericalLimits=true
    lODThreshold=1
    numSolverIterationsPostUpdate=4
    numSolverIterationsPreUpdate=16
    alphaBoolBlend {blendOption=Linear}
    boundBone {boneName=dyn_sharkTooth_01}
    physicsBodyDefinitions [{
      sphereCollisionRadius=0
      boundBone {boneName=dyn_sharkTooth_01}
      boxExtents=<10 10 1>
      constraintSetup {angularTargetAxis=AxisZ, twistAxis=AxisZ, angularLimitsMax=<30 0 0>, angularLimitsMin=<-30 0 0>, angularTarget=<0 0 0>}
      localJointOffset=<0 0 -1>
     }]
    sphericalLimits [{limitRadius=5, drivingBone {boneName=spine_05}, sphereLocalOffset=<1.5 5.5 0>}]
   }
   n3 = Anim Dynamics - Bone: dyn_necklace_01(ComponentPose=@n2, Alpha=1, ExternalForce="0,0,0") {
    angularBiasOverride=1
    angularDampingOverride=0.98
    bGravityOverrideInSimSpace=true
    bUseSphericalLimits=true
    lODThreshold=1
    numSolverIterationsPostUpdate=4
    numSolverIterationsPreUpdate=16
    alphaBoolBlend {blendOption=Linear}
    boundBone {boneName=dyn_necklace_01}
    physicsBodyDefinitions [{
      sphereCollisionRadius=0
      boundBone {boneName=dyn_necklace_01}
      boxExtents=<20 3 20>
      constraintSetup {angularTargetAxis=AxisZ, twistAxis=AxisZ, angularLimitsMax=<0 0 0>, angularLimitsMin=<0 0 0>, angularTarget=<0 0 0>, linearAxesMax=<0 2.5 0>}
      localJointOffset=<0 0 -1.5>
     }]
    sphericalLimits [{limitRadius=13.5, drivingBone {boneName=spine_04}, sphereLocalOffset=<1 0 0>}]
   }
   n4 = Anim Dynamics - Bone: dyn_necklace_02(ComponentPose=@n3, Alpha=1, ExternalForce="0,0,0") {
    angularBiasOverride=1
    angularDampingOverride=0.98
    bGravityOverrideInSimSpace=true
    bOverrideAngularBias=true
    bOverrideAngularDamping=true
    lODThreshold=1
    numSolverIterationsPostUpdate=4
    numSolverIterationsPreUpdate=16
    alphaBoolBlend {blendOption=Linear}
    boundBone {boneName=dyn_necklace_02}
    physicsBodyDefinitions [{
      sphereCollisionRadius=0
      boundBone {boneName=dyn_necklace_02}
      boxExtents=<10 2 10>
      constraintSetup {angularTargetAxis=AxisZ, twistAxis=AxisY, angularLimitsMax=<40 45 0>, angularLimitsMin=<-40 -45 0>, angularTarget=<0 0 0>}
      localJointOffset=<0 0 -10>
     }]
   }
   n5 = Anim Dynamics - Bone: dyn_necklace_03(ComponentPose=@n4, Alpha=1, ExternalForce="0,0,0") {
    angularBiasOverride=1
    angularDampingOverride=0.98
    bGravityOverrideInSimSpace=true
    bOverrideAngularBias=true
    bOverrideAngularDamping=true
    lODThreshold=1
    numSolverIterationsPostUpdate=4
    numSolverIterationsPreUpdate=16
    alphaBoolBlend {blendOption=Linear}
    boundBone {boneName=dyn_necklace_03}
    physicsBodyDefinitions [{
      sphereCollisionRadius=0
      boundBone {boneName=dyn_necklace_03}
      boxExtents=<10 2 10>
      constraintSetup {angularTargetAxis=AxisZ, twistAxis=AxisZ, angularLimitsMax=<0 45 0>, angularLimitsMin=<0 -45 0>, angularTarget=<0 0 0>}
      localJointOffset=<0 0 -1>
     }]
   }
   n6 = Anim Dynamics - Bone: dyn_pack_01(ComponentPose=@n5, Alpha=1, ExternalForce="0,0,0") {
    bGravityOverrideInSimSpace=true
    gravityScale=0.1
    lODThreshold=1
    alphaBoolBlend {blendOption=Linear}
    alphaScaleBiasClamp {bClampResult=true, clampMax=0.33, clampMin=0.1, scale=0.0033}
    boundBone {boneName=dyn_pack_01}
    physicsBodyDefinitions [{
      sphereCollisionRadius=0
      boundBone {boneName=dyn_pack_01}
      boxExtents=<5 5 5>
      constraintSetup {angularTargetAxis=AxisY, twistAxis=AxisY, angularLimitsMax=<0 0 0>, angularLimitsMin=<0 0 0>, angularTarget=<0 0 0>, linearAxesMax=<0 0 2>}
      localJointOffset=<0 0 5>
     }]
   }
   Output Pose(Result=Component To Local(ComponentPose=@n6)) {layerGroup=None}
   Get MeshToCopy() -> (MeshToCopy:Skeletal Mesh Component Object Reference)
  ]}
 EventGraph {nodes [
   Event BlueprintInitializeAnimation()
   > Set MeshToCopy(MeshToCopy:Skeletal Mesh Component Object Reference=GetOwnerSkeletalMeshComponent()) -> (Output_Get:Skeletal Mesh Component Object Reference)
   TryGetPawnOwner()
   Event BlueprintUpdateAnimation()
  ]}
]
variables [MeshToCopy {type=SkeletalMeshComponent}]
